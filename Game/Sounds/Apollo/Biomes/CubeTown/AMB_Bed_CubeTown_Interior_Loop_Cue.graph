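# Sound cue for the interior ambient bed of the CubeTown biome on the Apollo map.
# A single Wave Player is the root, playing a quad bed loop wave with no randomization or mixing; its looping flag is not set in the cue.
# It plays at about half volume through the Athena indoor ambient sound class, using a shared wide tile-ambient one-shot attenuation asset.
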SoundCue /Game/Sounds/Apollo/Biomes/CubeTown/AMB_Bed_CubeTown_Interior_Loop_Cue
root=WavePlayer_1
nodes [{id=WavePlayer_1, properties {soundWaveAssetPtr=&/AMB_Cubetown_Bed_Int_Loop_01_Quad}}]
properties {
 attenuationSettings=/Game/Sounds/Attenuation/Fort_TileAmbient_OneShot_Wider_Att
 maxDistance=8000
 soundClassObject=&Fort_Ambient_Indoors_Athena
 volumeMultiplier=0.501
 preEffectBusSends [{}]
}
