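# Sound cue for CubeTown biome outdoor ambience, playing one-shot "CubeCry" sounds with wide tile-ambient attenuation.
# A QualityLevel node sends its first two slots to a random pick of ten waves and its last two to a smaller four-wave set.
# Each branch passes through a Modulator that varies pitch and volume.
# The four-wave Random node is excluded from branch culling.
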SoundCue /Game/Sounds/Apollo/Biomes/CubeTown/AMB_SS_CubeTown_Cue
root=QualityLevel_0
nodes [
 {id=Modulator_0, children [Random_3], properties {pitchMax=1.1, pitchMin=0.7, volumeMax=1.1, volumeMin=0.9}}
 {id=Modulator_1, children [Random_1], properties {pitchMax=1.1, pitchMin=0.7, volumeMax=1.1, volumeMin=0.9}}
 {id=QualityLevel_0, children [Modulator_1,Modulator_1,Modulator_0,Modulator_0]}
 {
  id=Random_1
  children [
   WavePlayer_27
   WavePlayer_28
   WavePlayer_29
   WavePlayer_30
   WavePlayer_31
   WavePlayer_32
   WavePlayer_33
   WavePlayer_34
   WavePlayer_35
   WavePlayer_36
  ]
  properties {weights [1,1,1,1,1,1,1,1,1,1]}
 }
 {
  id=Random_3
  children [WavePlayer_37, WavePlayer_38, WavePlayer_39, WavePlayer_40]
  properties {bShouldExcludeFromBranchCulling=true, weights [1,1,1,1]}
 }
 {id=WavePlayer_27, properties {soundWaveAssetPtr=&/AMB_Cubetown_CubeCry_SS_01}}
 {id=WavePlayer_28, properties {soundWaveAssetPtr=&/AMB_Cubetown_CubeCry_SS_02}}
 {id=WavePlayer_29, properties {soundWaveAssetPtr=&/AMB_Cubetown_CubeCry_SS_03}}
 {id=WavePlayer_30, properties {soundWaveAssetPtr=&/AMB_Cubetown_CubeCry_SS_04}}
 {id=WavePlayer_31, properties {soundWaveAssetPtr=&/AMB_Cubetown_CubeCry_SS_05}}
 {id=WavePlayer_32, properties {soundWaveAssetPtr=&/AMB_Cubetown_CubeCry_SS_06}}
 {id=WavePlayer_33, properties {soundWaveAssetPtr=&/AMB_Cubetown_CubeCry_SS_07}}
 {id=WavePlayer_34, properties {soundWaveAssetPtr=&/AMB_Cubetown_CubeCry_SS_08}}
 {id=WavePlayer_35, properties {soundWaveAssetPtr=&/AMB_Cubetown_CubeCry_SS_09}}
 {id=WavePlayer_36, properties {soundWaveAssetPtr=&/AMB_Cubetown_CubeCry_SS_10}}
 {id=WavePlayer_37, properties {soundWaveAssetPtr=&AMB_Cubetown_CubeCry_SS_01}}
 {id=WavePlayer_38, properties {soundWaveAssetPtr=&AMB_Cubetown_CubeCry_SS_02}}
 {id=WavePlayer_39, properties {soundWaveAssetPtr=&AMB_Cubetown_CubeCry_SS_03}}
 {id=WavePlayer_40, properties {soundWaveAssetPtr=&AMB_Cubetown_CubeCry_SS_04}}
]
properties {
 attenuationSettings=/Game/Sounds/Attenuation/Fort_TileAmbient_OneShot_Wider_Att
 maxDistance=8000
 soundClassObject=&Fort_Ambient_Outdoors_Athena
 volumeMultiplier=1.002
 preEffectBusSends [{}]
}
